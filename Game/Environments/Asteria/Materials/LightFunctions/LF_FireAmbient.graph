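# Light function material for ambient firelight, combining an overall brightness flicker with scrolling noise.
# Flicker takes the max of three sines at different periods on randomly offset time, remapped above a minimum brightness.
# A low-res noise texture pans over long-lat UVs derived from the camera vector, with optional mirroring.
# Pole and seam masks brighten the noise to hide distortion, the seam mask only when horizontal tiling reaches one or more.
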Material /Game/Environments/Asteria/Materials/LightFunctions/LF_FireAmbient
graph {
 properties {materialDomain=MD_LightFunction, pixelDepthOffsetMode=PDOM_Legacy, usedShadingModels=Default Lit}
 nodes [
  n1 = MaterialFunctionCall(Float3=CameraVectorWS()) {materialFunction=&BreakOutFloat3Components}
  n2 = MaterialFunctionCall!(Vector=MaterialFunctionCall(X=Multiply(A=@n1.B) {constB=-1}, Y=@n1.G, Z=@n1.G) {materialFunction=&MakeFloat3}.Result) {materialFunction=/Engine/Functions/Engine_MaterialFunctions01/Coordinates/LongLatToUV}
  n3 = MaterialFunctionCall(Float2=@n2.Result) {materialFunction=&BreakOutFloat2Components}
  AppendVector(A=Frac(None=Subtract(A=@n3.R) {constB=0.125}), B=@n3.G)
  n4 = Add(A=Time(), B=ScalarParameter() {parameterName=RandomScalar_0-1})
  n5 = Multiply(A=@n4, B=ScalarParameter() {defaultValue=1, parameterName=FlickerSpeed})
  n6 = Sine(None=@n5) {period=0.8}
  n7 = Sine(None=@n5)
  MaterialExpressionRemap_1 = MaterialXRemap!(
   None=Max(A=Sine(None=@n5) {period=0.4}, B=Max(A=@n6, B=@n7))
   InputLow
   InputHigh
   TargetLow=ScalarParameter() {parameterName=FlickerMinBrightness}
   TargetHigh
  ) {inputLowDefault=-1}
  n8 = MaterialFunctionCall(
   UVs=@n2.Result
   Texture Scale=AppendVector(A=ScalarParameter() {defaultValue=1, parameterName=NoiseTilingX}, B=ScalarParameter() {defaultValue=1, parameterName=NoiseTilingY})
  ) {materialFunction=&ScaleUVsByCenter}
  n9 = MaterialFunctionCall(Float2=@n8.UVs) {materialFunction=&BreakOutFloat2Components}
  n10 = Abs(None=Subtract(A=@n9.R) {constB=0.5})
  n11 = StaticSwitchParameter(True=AppendVector(A=Multiply(A=@n10) {constB=2}, B=@n9.G), False=@n8.UVs) {
   desc="Mirrors uvs around the seam that occurs if NoiseTilingX>1.0 which may help hide it"
   group="Seam && Distortion Handling"
   parameterName=MirrorUVs
  }
  n12 = AppendVector(A=ScalarParameter() {defaultValue=0.25, parameterName=NoiseSpeedX}, B=ScalarParameter() {defaultValue=0.25, parameterName=NoiseSpeedY})
  n13 = TextureSampleParameter2D(UVs=Add(A=@n11, B=Multiply(A=@n4, B=@n12))) {
   automaticViewMipBias=false
   desc="Expected texture resolution is 64x64 pixels, no mips"
   parameterName=NoiseTexture
   samplerSource=SSM_Wrap_WorldGroupSettings
   samplerType=SAMPLERTYPE_LinearGrayscale
   texture=/Game/Global/Textures/Default/Noise/LowResBlurredNoise_NoMips
  }
  MaterialExpressionRemap_0 = MaterialXRemap(
   None=QualitySwitch(Default=Multiply(A=@n13.RGB), Low=@n13.RGB, Medium=@n13.RGB)
   TargetLow=ScalarParameter() {parameterName=NoiseMinBrightness}
  )
  n14 = Abs(None=ConstantBiasScale(None=@n2.Result) {bias=-0.5, scale=2})
  n15 = Power(
   Base=ComponentMask(None=@n14) {g=true}
   Exp=ScalarParameter() {defaultValue=1, group="Seam && Distortion Handling", parameterName=PoleMaskPower}
  )
  n16 = ComponentMask(None=@n14) {r=true}
  n17 = ScalarParameter() {defaultValue=2, desc="Only applies when NoiseTilingX>1.0", group="Seam && Distortion Handling", parameterName=SeamMaskPower}
  n18 = ScalarParameter() {defaultValue=1, desc="Only applies when NoiseTilingX>1.0", group="Seam && Distortion Handling", parameterName=SeamMaskIntensity}
  n19 = Multiply(A=SmoothStep(Value=Power(Base=@n16, Exp=@n17)), B=@n18)
  n20 = Subtract(A=ScalarParameter() {defaultValue=1, parameterName=NoiseTilingX}) {constB=0.9999}
  n21 = NamedRerouteDeclaration(None=Saturate(None=Sign(None=@n20))) {name=Seam, nodeColor=<rgba 0 1 0.25 1>}
  Multiply(A=@MaterialExpressionRemap_1, B=Max(A=Max(A=@MaterialExpressionRemap_0, B=@n15), B=Multiply(A=@n19, B=@n21)))
  TextureSample(UVs=TextureCoordinate() {uTiling=32, vTiling=32}) {samplerType=SAMPLERTYPE_LinearGrayscale, texture=&LowResBlurredNoise_NoMips}
 ]
}
